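# Physics asset for the left arm of the F_MED_Robot_Trouble female player body, covering its dynamic sleeve and glove cuff bones.
# Capsule bodies on upper arm, twist, forearm and hand form a chain with free angular constraints, with a large forearm capsule.
# Tightly limited constraints hang the cuff from the hand, upper-outer sleeve from the twist bone, and loose sleeve from the upper arm.
# Inner and outer sleeve bodies chain off the loose sleeve.
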PhysicsAsset /Game/Characters/Player/Female/Medium/Bodies/F_MED_Robot_Trouble/Meshes/F_MED_Robot_Trouble_Dyn_Arm_l
bodies {
 dyn_gloove_cuff_l [{shapeType=Capsule, length=3, radius=2, center=<0.1 0 0>, rotation=<rot 90 0 0>}]
 dyn_sleeve_in_l [{shapeType=Capsule, length=3, radius=2, center=<0.2 0 0>, rotation=<rot 90 0 0>}]
 dyn_sleeve_loose_l [{shapeType=Capsule, length=17.7143, radius=2.969, center=<0.2 0 0>, rotation=<rot 0 0 90>}]
 dyn_sleeve_out_l [{shapeType=Capsule, length=3, radius=2, center=<0.2 0 0>, rotation=<rot 90 0 0>}]
 dyn_sleeve_up_out_l [{shapeType=Capsule, length=3, radius=2, center=<0.2 0 0>, rotation=<rot 90 0 0>}]
 hand_l [{shapeType=Capsule, length=1, radius=1, rotation=<rot 90 0 0>}]
 lowerarm_l [{shapeType=Capsule, length=22.7637, radius=8.5929, center=<8.4989 -4.1331 0>, rotation=<rot -0.0001 179.9999 179.9999>}]
 upperarm_l [{shapeType=Capsule, length=1, radius=1, rotation=<rot 90 0 0>}]
 upperarm_twist_01_l [{shapeType=Capsule, length=1, radius=1, rotation=<rot 90 0 0>}]
}
constraints [
 {bone1Name=dyn_gloove_cuff_l, bone2Name=hand_l, swing1LimitDegrees=20, swing2LimitDegrees=20, twistLimitDegrees=2}
 {bone1Name=dyn_sleeve_up_out_l, bone2Name=upperarm_twist_01_l, swing1LimitDegrees=10, swing2LimitDegrees=20, twistLimitDegrees=2}
 {bone1Name=dyn_sleeve_in_l, bone2Name=dyn_sleeve_loose_l, swing1LimitDegrees=10, swing2LimitDegrees=20, twistLimitDegrees=2}
 {bone1Name=dyn_sleeve_out_l, bone2Name=dyn_sleeve_loose_l, swing1LimitDegrees=10, swing2LimitDegrees=10, twistLimitDegrees=2}
 {
  bone1Name=hand_l
  bone2Name=lowerarm_l
  swing1LimitDegrees=45
  swing1Motion=Free
  swing2LimitDegrees=45
  swing2Motion=Free
  twistLimitDegrees=45
  twistMotion=Free
 }
 {
  bone1Name=upperarm_twist_01_l
  bone2Name=upperarm_l
  swing1LimitDegrees=45
  swing1Motion=Free
  swing2LimitDegrees=45
  swing2Motion=Free
  twistLimitDegrees=45
  twistMotion=Free
 }
 {bone1Name=dyn_sleeve_loose_l, bone2Name=upperarm_l, swing1LimitDegrees=20, swing2LimitDegrees=20, twistLimitDegrees=10}
 {
  bone1Name=lowerarm_l
  bone2Name=upperarm_l
  swing1LimitDegrees=45
  swing1Motion=Free
  swing2LimitDegrees=45
  swing2Motion=Free
  twistLimitDegrees=45
  twistMotion=Free
 }
]
