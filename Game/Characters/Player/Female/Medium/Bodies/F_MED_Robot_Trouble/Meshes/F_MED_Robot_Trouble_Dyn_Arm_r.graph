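# Physics asset for the right-arm dynamic bones of the medium female Robot Trouble player body, covering the sleeve and glove cuff.
# Capsule bodies sit on the inner sleeve, outer sleeve and cuff bones, with smaller capsules on the hand and two upper-arm twist bones.
# Narrow swing and very tight twist limits tie the inner sleeve to upperarm_twist_02_r, the outer sleeve to upperarm_twist_01_r, and the cuff to hand_r.
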PhysicsAsset /Game/Characters/Player/Female/Medium/Bodies/F_MED_Robot_Trouble/Meshes/F_MED_Robot_Trouble_Dyn_Arm_r
bodies {
 dyn_gloove_cuff_r [{shapeType=Capsule, length=3, radius=2, center=<0.1 0 0>, rotation=<rot 90 0 0>}]
 dyn_sleeve_in_r [{shapeType=Capsule, length=3, radius=2, center=<-0.4 0 0>, rotation=<rot 90 0 0>}]
 dyn_sleeve_out_r [{shapeType=Capsule, length=3, radius=2, center=<-0.4 0 0>, rotation=<rot 90 0 0>}]
 hand_r [{shapeType=Capsule, length=1, radius=1, rotation=<rot 90 0 0>}]
 upperarm_twist_01_r [{shapeType=Capsule, length=1, radius=1, rotation=<rot 90 0 0>}]
 upperarm_twist_02_r [{shapeType=Capsule, length=1, radius=1, rotation=<rot 90 0 0>}]
}
constraints [
 {bone1Name=dyn_sleeve_in_r, bone2Name=upperarm_twist_02_r, swing1LimitDegrees=10, swing2LimitDegrees=20, twistLimitDegrees=2}
 {bone1Name=dyn_sleeve_out_r, bone2Name=upperarm_twist_01_r, swing1LimitDegrees=10, swing2LimitDegrees=20, twistLimitDegrees=2}
 {bone1Name=dyn_gloove_cuff_r, bone2Name=hand_r, swing1LimitDegrees=10, swing2LimitDegrees=10, twistLimitDegrees=2}
]
