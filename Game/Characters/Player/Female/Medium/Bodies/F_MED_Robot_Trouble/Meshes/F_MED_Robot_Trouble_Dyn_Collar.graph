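# Physics asset for the dynamic collar on the medium female Robot Trouble player body, covering the collar, head, upper arms and upper spine.
# Back, left and right collar flaps are each built from three angled boxes, alongside capsules on the head, upper arms and spine_05.
# The back flap constrains to spine_05, while the side flaps constrain to small MSTR bodies.
# Those MSTR bodies, the head and the arms join spine_05 with free angular motion.
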PhysicsAsset /Game/Characters/Player/Female/Medium/Bodies/F_MED_Robot_Trouble/Meshes/F_MED_Robot_Trouble_Dyn_Collar
bodies {
 dyn_collar_back [
  {shapeType=Box, extentX=12.3913, extentY=6, extentZ=10, center=<5.6071 -2.2901 0.0001>, rotation=<rot 0 -5.875 0>}
  {shapeType=Box, extentX=12.9555, extentY=6, extentZ=11.7544, center=<4.8326 -1.3443 -5.1697>, rotation=<rot -9.6755 -1.7123 -20.7255>}
  {shapeType=Box, extentX=11.6585, extentY=6, extentZ=10, center=<5.4988 -0.953 6.0613>, rotation=<rot 8.5132 -4.7414 26.05>}
 ]
 dyn_collar_l [
  {shapeType=Box, extentX=11.1883, extentY=10, extentZ=1.9235, center=<5.2572 0.2235 -3.302>, rotation=<rot -9.7764 12.26 -2.113>}
  {shapeType=Box, extentX=10.9239, extentY=13.5847, extentZ=1.9135, center=<6.7561 -4.7532 -0.1987>, rotation=<rot -15.9218 -5.9405 40.8947>}
  {shapeType=Box, extentX=10.2697, extentY=10, extentZ=2.0371, center=<4.0937 5.5784 -2.7802>, rotation=<rot -14.4397 20.716 -26.1062>}
 ]
 dyn_collar_l_MSTR [{shapeType=Capsule, length=0.505, radius=0.505, rotation=<rot 0 -90 0>}]
 dyn_collar_r [
  {shapeType=Box, extentX=12.3912, extentY=10, extentZ=3, center=<4.6756 0.1485 3.8373>, rotation=<rot 10.1406 12.0271 -6.0845>}
  {shapeType=Box, extentX=12.5368, extentY=27.701, extentZ=4.5, center=<5.605 -3.2036 3.2826>, rotation=<rot 10.5853 -2.3718 -31.6838>}
  {shapeType=Box, extentX=10.5993, extentY=10, extentZ=3, center=<3.9911 5.5667 2.6994>, rotation=<rot 18.5663 22.0756 24.2453>}
 ]
 dyn_collar_r_MSTR [{shapeType=Capsule, length=0.505, radius=0.505, rotation=<rot 0 -90 0>}]
 head [
  {shapeType=Capsule, length=4.4105, radius=10.9756, center=<8.8863 0.8074 -0.0052>, rotation=<rot 0.4679 -98.1592 89.9271>}
  {shapeType=Capsule, length=3.8705, radius=1.2279, center=<-6.2517 7.656 -5.2874>, rotation=<rot 0 0 60.2346>}
 ]
 spine_05 [{shapeType=Capsule, length=1, radius=1}]
 upperarm_l [{shapeType=Capsule, length=7.7829, radius=5.1005, center=<4.3732 -0.0003 0.8555>, rotation=<rot 82.4091 0 0>}]
 upperarm_r [{shapeType=Capsule, length=7.2099, radius=5.6735, center=<-9.3801 0.0003 0>, rotation=<rot 90 0 0>}]
}
constraints [
 {
  bone1Name=head
  bone2Name=spine_05
  swing1LimitDegrees=45
  swing1Motion=Free
  swing2LimitDegrees=45
  swing2Motion=Free
  twistLimitDegrees=45
  twistMotion=Free
 }
 {bone1Name=dyn_collar_r, bone2Name=dyn_collar_r_MSTR, swing1LimitDegrees=90, swing2LimitDegrees=90, twistLimitDegrees=90}
 {bone1Name=dyn_collar_back, bone2Name=spine_05, swing1LimitDegrees=90, swing2LimitDegrees=90, twistLimitDegrees=50}
 {bone1Name=dyn_collar_l, bone2Name=dyn_collar_l_MSTR, swing1LimitDegrees=90, swing2LimitDegrees=90, twistLimitDegrees=50}
 {
  bone1Name=upperarm_r
  bone2Name=spine_05
  swing1LimitDegrees=45
  swing1Motion=Free
  swing2LimitDegrees=45
  swing2Motion=Free
  twistLimitDegrees=45
  twistMotion=Free
 }
 {
  bone1Name=upperarm_l
  bone2Name=spine_05
  swing1LimitDegrees=45
  swing1Motion=Free
  swing2LimitDegrees=45
  swing2Motion=Free
  twistLimitDegrees=45
  twistMotion=Free
 }
 {
  bone1Name=dyn_collar_l_MSTR
  bone2Name=spine_05
  swing1LimitDegrees=45
  swing1Motion=Free
  swing2LimitDegrees=45
  swing2Motion=Free
  twistLimitDegrees=45
  twistMotion=Free
 }
 {
  bone1Name=dyn_collar_r_MSTR
  bone2Name=spine_05
  swing1LimitDegrees=45
  swing1Motion=Free
  swing2LimitDegrees=45
  swing2Motion=Free
  twistLimitDegrees=45
  twistMotion=Free
 }
]
