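# Physics asset for the dynamic jacket of the Robot Trouble medium female player body, defining capsule bodies and constraints for secondary motion.
# Seven two-bone chains around front, sides and back hang from spine_04, with narrow twist and per-chain swing limits.
# The right-side chain alone attaches through an intermediate MSTR body, whose spine link has free rotation.
# Pelvis, spine_04 and spine_05 supply torso bodies, also linked with free rotation.
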PhysicsAsset /Game/Characters/Player/Female/Medium/Bodies/F_MED_Robot_Trouble/Meshes/F_MED_Robot_Trouble_Dyn_Jacket
bodies {
 dyn_jacket_bk_01 [{shapeType=Capsule, length=3, radius=1, center=<1.5 0 0>, rotation=<rot 90 0 0>}]
 dyn_jacket_bk_02 [{shapeType=Capsule, length=3, radius=1, center=<1.5 0 0>, rotation=<rot 90 0 0>}]
 dyn_jacket_bk_l_01 [{shapeType=Capsule, length=3, radius=1, center=<1.5 0 0>, rotation=<rot 90 0 0>}]
 dyn_jacket_bk_l_02 [{shapeType=Capsule, length=3, radius=1, center=<1.5 0 0>, rotation=<rot 90 0 0>}]
 dyn_jacket_bk_r_01 [{shapeType=Capsule, length=3, radius=1, center=<-1.5 0 0>, rotation=<rot 90 0 0>}]
 dyn_jacket_bk_r_02 [{shapeType=Capsule, length=3, radius=1, center=<-1.5 0 0>, rotation=<rot 90 0 0>}]
 dyn_jacket_fr_l_01 [{shapeType=Capsule, length=3, radius=1, center=<1.5 0 0>, rotation=<rot 90 0 0>}]
 dyn_jacket_fr_l_02 [{shapeType=Capsule, length=3, radius=1, center=<1.5 -0.0001 -0.4875>, rotation=<rot 90 0 0>}]
 dyn_jacket_fr_r_01 [{shapeType=Capsule, length=3, radius=1, center=<-1.5 0 0>, rotation=<rot 90 0 0>}]
 dyn_jacket_fr_r_02 [{shapeType=Capsule, length=3, radius=1, center=<-1.5 0 0>, rotation=<rot 90 0 0>}]
 dyn_jacket_l_01 [{shapeType=Capsule, length=3, radius=1, center=<1.5 0 0>, rotation=<rot 90 0 0>}]
 dyn_jacket_l_02 [{shapeType=Capsule, length=3, radius=1, center=<1.5 0 0>, rotation=<rot 90 0 0>}]
 dyn_jacket_r_01 [{shapeType=Capsule, length=3, radius=1, center=<-1.5 0 0>, rotation=<rot 90 0 0>}]
 dyn_jacket_r_01_MSTR [{shapeType=Capsule, length=0.505, radius=0.505, rotation=<rot 0 -90 0>}]
 dyn_jacket_r_02 [{shapeType=Capsule, length=3, radius=1, center=<-1.5 0 0>, rotation=<rot 90 0 0>}]
 pelvis [{shapeType=Capsule, length=3, radius=3}]
 spine_04 [
  {shapeType=Capsule, length=3.5187, radius=9.25, center=<1.37 3.768 0>}
  {shapeType=Capsule, length=3.75, radius=9.25, center=<0.3228 -0.457 0>}
 ]
 spine_05 [{shapeType=Capsule, length=1, radius=1}]
}
constraints [
 {bone1Name=dyn_jacket_bk_02, bone2Name=dyn_jacket_bk_01, swing1LimitDegrees=45, swing2LimitDegrees=30, twistLimitDegrees=5}
 {bone1Name=dyn_jacket_bk_l_02, bone2Name=dyn_jacket_bk_l_01, swing1LimitDegrees=45, swing2LimitDegrees=30, twistLimitDegrees=5}
 {bone1Name=dyn_jacket_l_02, bone2Name=dyn_jacket_l_01, swing1LimitDegrees=45, swing2LimitDegrees=10, twistLimitDegrees=5}
 {bone1Name=dyn_jacket_fr_l_02, bone2Name=dyn_jacket_fr_l_01, swing1LimitDegrees=45, swing2LimitDegrees=50, twistLimitDegrees=5}
 {bone1Name=dyn_jacket_r_02, bone2Name=dyn_jacket_r_01, swing1LimitDegrees=5, swing2LimitDegrees=0, twistLimitDegrees=5}
 {bone1Name=dyn_jacket_fr_r_02, bone2Name=dyn_jacket_fr_r_01, swing1LimitDegrees=45, swing2LimitDegrees=50, twistLimitDegrees=5}
 {bone1Name=dyn_jacket_bk_r_02, bone2Name=dyn_jacket_bk_r_01, swing1LimitDegrees=10, swing2LimitDegrees=10, twistLimitDegrees=5}
 {bone1Name=dyn_jacket_bk_01, bone2Name=spine_04, swing1LimitDegrees=45, swing2LimitDegrees=30, twistLimitDegrees=5}
 {bone1Name=dyn_jacket_bk_l_01, bone2Name=spine_04, swing1LimitDegrees=45, swing2LimitDegrees=30, twistLimitDegrees=5}
 {bone1Name=dyn_jacket_l_01, bone2Name=spine_04, swing1LimitDegrees=45, swing2LimitDegrees=10, twistLimitDegrees=5}
 {bone1Name=dyn_jacket_fr_l_01, bone2Name=spine_04, swing1LimitDegrees=5, swing2LimitDegrees=50, twistLimitDegrees=5}
 {bone1Name=dyn_jacket_r_01, bone2Name=dyn_jacket_r_01_MSTR, swing1LimitDegrees=5, swing2LimitDegrees=5, twistLimitDegrees=5}
 {bone1Name=dyn_jacket_fr_r_01, bone2Name=spine_04, swing1LimitDegrees=5, swing2LimitDegrees=50, twistLimitDegrees=5}
 {bone1Name=dyn_jacket_bk_r_01, bone2Name=spine_04, swing1LimitDegrees=10, swing2LimitDegrees=5, twistLimitDegrees=5}
 {
  bone1Name=spine_04
  bone2Name=pelvis
  swing1LimitDegrees=45
  swing1Motion=Free
  swing2LimitDegrees=45
  swing2Motion=Free
  twistLimitDegrees=45
  twistMotion=Free
 }
 {
  bone1Name=spine_05
  bone2Name=spine_04
  swing1LimitDegrees=45
  swing1Motion=Free
  swing2LimitDegrees=45
  swing2Motion=Free
  twistLimitDegrees=45
  twistMotion=Free
 }
 {
  bone1Name=dyn_jacket_r_01_MSTR
  bone2Name=spine_04
  swing1LimitDegrees=45
  swing1Motion=Free
  swing2LimitDegrees=45
  swing2Motion=Free
  twistLimitDegrees=45
  twistMotion=Free
 }
]
