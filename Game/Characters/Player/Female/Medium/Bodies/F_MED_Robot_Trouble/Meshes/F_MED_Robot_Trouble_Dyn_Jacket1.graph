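# Physics asset that drives secondary motion of the jacket on the F_MED_Robot_Trouble female player body.
# Seven two-bone capsule chains around back, sides and front hang from spine_04, with torso capsules on pelvis, spine_04 and spine_05.
# Per-joint swing and twist limits differ between left and right, and the back-right root is angularly locked to the spine.
# Angularly free cross constraints link neighboring lower back and side bones, tying those chains together.
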PhysicsAsset /Game/Characters/Player/Female/Medium/Bodies/F_MED_Robot_Trouble/Meshes/F_MED_Robot_Trouble_Dyn_Jacket1
bodies {
 dyn_jacket_bk_01 [{shapeType=Capsule, length=3.625, radius=1.25, center=<1.7749 0 0>, rotation=<rot 90 0 0>}]
 dyn_jacket_bk_02 [{shapeType=Capsule, length=4.375, radius=1.5, center=<1.7749 0 0>, rotation=<rot 90 0 0>}]
 dyn_jacket_bk_l_01 [{shapeType=Capsule, length=2.875, radius=1.125, center=<1.5 0 0>, rotation=<rot 90 0 0>}]
 dyn_jacket_bk_l_02 [{shapeType=Capsule, length=3.625, radius=1.5, center=<1.5 0 0>, rotation=<rot 90 0 0>}]
 dyn_jacket_bk_r_01 [{shapeType=Capsule, length=2.875, radius=1.125, center=<-1.5 0 0>, rotation=<rot 90 0 0>}]
 dyn_jacket_bk_r_02 [{shapeType=Capsule, length=3.625, radius=1.5, center=<-1.5 0 0>, rotation=<rot 90 0 0>}]
 dyn_jacket_fr_l_01 [{shapeType=Capsule, length=3, radius=1, center=<1.5 0 0>, rotation=<rot 90 0 0>}]
 dyn_jacket_fr_l_02 [{shapeType=Capsule, length=3, radius=1.125, center=<1.5 -0.0001 -0.4875>, rotation=<rot 90 0 0>}]
 dyn_jacket_fr_r_01 [{shapeType=Capsule, length=3, radius=1, center=<-1.5 0 0>, rotation=<rot 90 0 0>}]
 dyn_jacket_fr_r_02 [{shapeType=Capsule, length=3.375, radius=1.25, center=<-1.5 0 0>, rotation=<rot 90 0 0>}]
 dyn_jacket_l_01 [{shapeType=Capsule, length=3, radius=1, center=<1.5 0 0>, rotation=<rot 90 0 0>}]
 dyn_jacket_l_02 [{shapeType=Capsule, length=2.875, radius=1.125, center=<1.5 0 0>, rotation=<rot 90 0 0>}]
 dyn_jacket_r_01 [{shapeType=Capsule, length=3, radius=1, center=<-1.5 0 0>, rotation=<rot 90 0 0>}]
 dyn_jacket_r_02 [{shapeType=Capsule, length=2.875, radius=1.125, center=<-1.5 0 0>, rotation=<rot 90 0 0>}]
 pelvis [{shapeType=Capsule, length=3, radius=3}]
 spine_04 [
  {shapeType=Capsule, length=4.0187, radius=9.25, center=<1.37 3.768 0>}
  {shapeType=Capsule, length=4.125, radius=9.25, center=<-1.7998 -0.0564 0>}
 ]
 spine_05 [{shapeType=Capsule, length=1, radius=1.125}]
}
constraints [
 {bone1Name=dyn_jacket_bk_02, bone2Name=dyn_jacket_bk_01, swing1LimitDegrees=25, swing2LimitDegrees=40, twistLimitDegrees=5}
 {bone1Name=dyn_jacket_bk_l_02, bone2Name=dyn_jacket_bk_l_01, swing1LimitDegrees=45, swing2LimitDegrees=30, twistLimitDegrees=5}
 {bone1Name=dyn_jacket_l_02, bone2Name=dyn_jacket_l_01, swing1LimitDegrees=45, swing2LimitDegrees=10, twistLimitDegrees=5}
 {bone1Name=dyn_jacket_fr_l_02, bone2Name=dyn_jacket_fr_l_01, swing1LimitDegrees=45, swing2LimitDegrees=50, twistLimitDegrees=5}
 {bone1Name=dyn_jacket_r_02, bone2Name=dyn_jacket_r_01, swing1LimitDegrees=45, swing2LimitDegrees=5, twistLimitDegrees=5}
 {bone1Name=dyn_jacket_fr_r_02, bone2Name=dyn_jacket_fr_r_01, swing1LimitDegrees=45, swing2LimitDegrees=50, twistLimitDegrees=5}
 {bone1Name=dyn_jacket_bk_r_02, bone2Name=dyn_jacket_bk_r_01, swing1LimitDegrees=15, swing2LimitDegrees=15, twistLimitDegrees=5}
 {bone1Name=dyn_jacket_bk_01, bone2Name=spine_04, swing1LimitDegrees=25, swing2LimitDegrees=25, twistLimitDegrees=5}
 {bone1Name=dyn_jacket_bk_l_01, bone2Name=spine_04, swing1LimitDegrees=25, swing2LimitDegrees=10, twistLimitDegrees=5}
 {bone1Name=dyn_jacket_l_01, bone2Name=spine_04, swing1LimitDegrees=30, swing2LimitDegrees=15, twistLimitDegrees=5}
 {bone1Name=dyn_jacket_fr_l_01, bone2Name=spine_04, swing1LimitDegrees=5, swing2LimitDegrees=25, twistLimitDegrees=5}
 {bone1Name=dyn_jacket_r_01, bone2Name=spine_04, swing1LimitDegrees=5, swing2LimitDegrees=5, twistLimitDegrees=5}
 {bone1Name=dyn_jacket_fr_r_01, bone2Name=spine_04, swing1LimitDegrees=5, swing2LimitDegrees=25, twistLimitDegrees=5}
 {
  bone1Name=dyn_jacket_bk_r_01
  bone2Name=spine_04
  swing1LimitDegrees=3
  swing1Motion=Locked
  swing2LimitDegrees=5
  swing2Motion=Locked
  twistLimitDegrees=5
  twistMotion=Locked
 }
 {
  bone1Name=spine_04
  bone2Name=pelvis
  swing1LimitDegrees=45
  swing1Motion=Free
  swing2LimitDegrees=45
  swing2Motion=Free
  twistLimitDegrees=45
  twistMotion=Free
 }
 {
  bone1Name=spine_05
  bone2Name=spine_04
  swing1LimitDegrees=45
  swing1Motion=Free
  swing2LimitDegrees=45
  swing2Motion=Free
  twistLimitDegrees=45
  twistMotion=Free
 }
 {
  bone1Name=dyn_jacket_l_02
  bone2Name=dyn_jacket_bk_l_02
  swing1LimitDegrees=45
  swing1Motion=Free
  swing2LimitDegrees=45
  swing2Motion=Free
  twistLimitDegrees=45
  twistMotion=Free
 }
 {
  bone1Name=dyn_jacket_r_02
  bone2Name=dyn_jacket_bk_r_02
  swing1LimitDegrees=45
  swing1Motion=Free
  swing2LimitDegrees=45
  swing2Motion=Free
  twistLimitDegrees=45
  twistMotion=Free
 }
 {
  bone1Name=dyn_jacket_bk_r_02
  bone2Name=dyn_jacket_bk_02
  swing1LimitDegrees=45
  swing1Motion=Free
  swing2LimitDegrees=45
  swing2Motion=Free
  twistLimitDegrees=45
  twistMotion=Free
 }
 {
  bone1Name=dyn_jacket_bk_l_02
  bone2Name=dyn_jacket_bk_02
  swing1LimitDegrees=45
  swing1Motion=Free
  swing2LimitDegrees=45
  swing2Motion=Free
  twistLimitDegrees=45
  twistMotion=Free
 }
]
