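# Physics asset for the left-leg dynamic pouches on the female medium Robot Trouble player body.
# Small capsule bodies cover the calf and thigh pouch bones and the twist bones they hang from.
# Each pouch is constrained to its twist bone with moderate swing and tight twist limits, so it can sway while barely rotating about its axis.
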PhysicsAsset /Game/Characters/Player/Female/Medium/Bodies/F_MED_Robot_Trouble/Meshes/F_MED_Robot_Trouble_Dyn_Leg_l
bodies {
 calf_twist_02_l [{shapeType=Capsule, length=2, radius=3, rotation=<rot 90 0 0>}]
 dyn_calf_pouch_l [{shapeType=Capsule, length=2, radius=3, center=<2 0 0>, rotation=<rot 90 0 0>}]
 dyn_thigh_pouch_l [{shapeType=Capsule, length=2, radius=3, center=<2 0 0>, rotation=<rot 90 0 0>}]
 thigh_twist_01_l [{shapeType=Capsule, length=2, radius=3, rotation=<rot 90 0 0>}]
}
constraints [
 {bone1Name=dyn_calf_pouch_l, bone2Name=calf_twist_02_l, swing1LimitDegrees=45, swing2LimitDegrees=45, twistLimitDegrees=5}
 {bone1Name=dyn_thigh_pouch_l, bone2Name=thigh_twist_01_l, swing1LimitDegrees=45, swing2LimitDegrees=45, twistLimitDegrees=5}
]
