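# Physics asset for the dynamic shoulder straps and satchel on the medium female Robot Trouble player body.
# Small master bodies on each clavicle anchor a two-segment front strap chain and a single, more tightly limited back strap.
# The satchel box hangs from the pelvis with limited swing and tight twist.
# Spine capsules and a box form the torso volume, while body-to-body and master-to-clavicle joints leave rotation free.
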PhysicsAsset /Game/Characters/Player/Female/Medium/Bodies/F_MED_Robot_Trouble/Meshes/F_MED_Robot_Trouble_Dyn_Straps
bodies {
 clavicle_l [{shapeType=Capsule, length=3, radius=2, rotation=<rot 90 0 0>}]
 clavicle_r [{shapeType=Capsule, length=3, radius=2, rotation=<rot 90 0 0>}]
 dyn_satchel [{shapeType=Box, extentX=12.5, extentY=21, extentZ=5, center=<0.7386 0.1302 -0.1698>, rotation=<rot 0 10.002 0>}]
 dyn_shoulder_strap_bk_l [{shapeType=Capsule, length=3, radius=2, center=<0.5151 0 0>, rotation=<rot 90 0 0>}]
 dyn_shoulder_strap_bk_l_MSTR [{shapeType=Capsule, length=0.505, radius=0.505, rotation=<rot 0 -90 0>}]
 dyn_shoulder_strap_bk_r [{shapeType=Capsule, length=3, radius=2, center=<-2.4844 -0.0001 0>, rotation=<rot 90 0 0>}]
 dyn_shoulder_strap_bk_r_MSTR [{shapeType=Capsule, length=0.505, radius=0.505, rotation=<rot 0 -90 0>}]
 dyn_shoulder_strap_fr_l_01 [{
   shapeType=Capsule
   length=2.3294
   radius=1.6528
   shapeName=dyn_shoulder_strap_fr_l_01_capsule
   center=<3.0757 0.1243 0.0918>
   rotation=<rot 90 0 0>
  }]
 dyn_shoulder_strap_fr_l_02 [{
   shapeType=Capsule
   length=3.047
   radius=1.7208
   shapeName=dyn_shoulder_strap_fr_l_02_capsule
   center=<3.1826 0 0.0001>
   rotation=<rot 90 0 0>
  }]
 dyn_shoulder_strap_fr_l_MSTR [{shapeType=Capsule, length=0.505, radius=0.505, rotation=<rot 0 -90 0>}]
 dyn_shoulder_strap_fr_r_01 [{
   shapeType=Capsule
   length=2.2476
   radius=1.3341
   shapeName=dyn_shoulder_strap_fr_r_01_capsule
   center=<-2.5665 -0.0631 -0.0542>
   rotation=<rot 90 0 0>
  }]
 dyn_shoulder_strap_fr_r_02 [{
   shapeType=Capsule
   length=2.7897
   radius=1.252
   shapeName=dyn_shoulder_strap_fr_r_02_capsule
   center=<-3.1451 0.3813 0.5027>
   rotation=<rot 0 -99.8524 -80.2293>
  }]
 dyn_shoulder_strap_fr_r_MSTR [{shapeType=Capsule, length=0.505, radius=0.505, rotation=<rot 0 -90 0>}]
 pelvis [{shapeType=Capsule, length=3, radius=3}]
 spine_04 [
  {shapeType=Capsule, length=16.875, radius=15.0625, center=<-0.2761 -4.276 0>}
  {shapeType=Box, extentX=18.875, extentY=5.6875, extentZ=22.5, center=<-3.3355 -2.5269 -0.0001>, rotation=<rot 0 -35.0003 0>}
 ]
 spine_05 [{shapeType=Capsule, length=18.4143, radius=11.6126, shapeName=spine_05_capsule, center=<-6.0438 1.2298 0>}]
 upperarm_twist_01_l [{shapeType=Capsule, length=1, radius=1}]
}
constraints [
 {bone1Name=dyn_shoulder_strap_bk_r, bone2Name=dyn_shoulder_strap_bk_r_MSTR, swing1LimitDegrees=5, swing2LimitDegrees=30, twistLimitDegrees=5}
 {bone1Name=dyn_shoulder_strap_bk_l, bone2Name=dyn_shoulder_strap_bk_l_MSTR, swing1LimitDegrees=5, swing2LimitDegrees=30, twistLimitDegrees=5}
 {
  bone1Name=clavicle_l
  bone2Name=spine_05
  swing1LimitDegrees=45
  swing1Motion=Free
  swing2LimitDegrees=45
  swing2Motion=Free
  twistLimitDegrees=45
  twistMotion=Free
 }
 {
  bone1Name=clavicle_r
  bone2Name=spine_05
  swing1LimitDegrees=45
  swing1Motion=Free
  swing2LimitDegrees=45
  swing2Motion=Free
  twistLimitDegrees=45
  twistMotion=Free
 }
 {
  bone1Name=spine_04
  bone2Name=pelvis
  swing1LimitDegrees=45
  swing1Motion=Free
  swing2LimitDegrees=45
  swing2Motion=Free
  twistLimitDegrees=45
  twistMotion=Free
 }
 {bone1Name=dyn_satchel, bone2Name=pelvis, swing1LimitDegrees=45, swing2LimitDegrees=45, twistLimitDegrees=5}
 {
  bone1Name=spine_05
  bone2Name=spine_04
  swing1LimitDegrees=45
  swing1Motion=Free
  swing2LimitDegrees=45
  swing2Motion=Free
  twistLimitDegrees=45
  twistMotion=Free
 }
 {
  bone1Name=upperarm_twist_01_l
  bone2Name=clavicle_l
  swing1LimitDegrees=45
  swing1Motion=Free
  swing2LimitDegrees=45
  swing2Motion=Free
  twistLimitDegrees=45
  twistMotion=Free
 }
 {
  bone1Name=dyn_shoulder_strap_fr_r_01
  bone2Name=dyn_shoulder_strap_fr_r_MSTR
  swing1LimitDegrees=45
  swing2LimitDegrees=45
  twistLimitDegrees=30
 }
 {
  bone1Name=dyn_shoulder_strap_fr_r_02
  bone2Name=dyn_shoulder_strap_fr_r_01
  swing1LimitDegrees=45
  swing2LimitDegrees=45
  twistLimitDegrees=30
 }
 {
  bone1Name=dyn_shoulder_strap_fr_l_01
  bone2Name=dyn_shoulder_strap_fr_l_MSTR
  swing1LimitDegrees=45
  swing2LimitDegrees=45
  twistLimitDegrees=30
 }
 {
  bone1Name=dyn_shoulder_strap_fr_l_02
  bone2Name=dyn_shoulder_strap_fr_l_01
  swing1LimitDegrees=45
  swing2LimitDegrees=45
  twistLimitDegrees=30
 }
 {
  bone1Name=dyn_shoulder_strap_bk_l_MSTR
  bone2Name=clavicle_l
  swing1LimitDegrees=45
  swing1Motion=Free
  swing2LimitDegrees=45
  swing2Motion=Free
  twistLimitDegrees=45
  twistMotion=Free
 }
 {
  bone1Name=dyn_shoulder_strap_fr_l_MSTR
  bone2Name=clavicle_l
  swing1LimitDegrees=45
  swing1Motion=Free
  swing2LimitDegrees=45
  swing2Motion=Free
  twistLimitDegrees=45
  twistMotion=Free
 }
 {
  bone1Name=dyn_shoulder_strap_bk_r_MSTR
  bone2Name=clavicle_r
  swing1LimitDegrees=45
  swing1Motion=Free
  swing2LimitDegrees=45
  swing2Motion=Free
  twistLimitDegrees=45
  twistMotion=Free
 }
 {
  bone1Name=dyn_shoulder_strap_fr_r_MSTR
  bone2Name=clavicle_r
  swing1LimitDegrees=45
  swing1Motion=Free
  swing2LimitDegrees=45
  swing2Motion=Free
  twistLimitDegrees=45
  twistMotion=Free
 }
]
