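# Secondary-motion AnimBP for the female Robot Trouble player body: jacket, collar, straps, pouches and mechanical joints.
# Copies the attached parent's pose, solves pistons, elbows and constraints, then caches it pre-simulation.
# bRigidBodyEnabled selects RigidBody and AnimDynamics simulation with corrective constraints, or a constraint-only fallback.
# The EventGraph derives corrective alphas from jump, DBNO, crouch, arm and neck angles; some older simulation chains are unconnected.
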AnimBlueprint /Game/Characters/Player/Female/Medium/Bodies/F_MED_Robot_Trouble/Meshes/F_MED_Robot_Trouble_AnimBP
target_skeleton=&/F_MED_Robot_Trouble_Skeleton
graphs [
 AnimGraph {nodes [
   n1 = Get DBNO Alpha() -> (DBNO Alpha:double)
   n2 = "Constraint - dyn_jacket_bk_l_01 (P)"(Alpha=@n1, ConstraintWeights_0=1) {
    alphaBoolBlend {blendOption=Linear}
    boneToModify {boneName=dyn_jacket_bk_l_01}
    constraintSetup [{targetBone {boneName=spine_04}}]
    constraintWeights [1]
   }
   n3 = "Constraint - dyn_jacket_bk_l_02 (P)"(ComponentPose=@n2, Alpha=@n1, ConstraintWeights_0=1) {
    alphaBoolBlend {blendOption=Linear}
    boneToModify {boneName=dyn_jacket_bk_l_02}
    constraintSetup [{targetBone {boneName=spine_04}}]
    constraintWeights [1]
   }
   n4 = "Constraint - dyn_jacket_bk_r_01 (P)"(ComponentPose=@n3, Alpha=@n1, ConstraintWeights_0=1) {
    alphaBoolBlend {blendOption=Linear}
    boneToModify {boneName=dyn_jacket_bk_r_01}
    constraintSetup [{targetBone {boneName=spine_04}}]
    constraintWeights [1]
   }
   n5 = "Constraint - dyn_jacket_bk_r_02 (P)"(ComponentPose=@n4, Alpha=@n1, ConstraintWeights_0=1) {
    alphaBoolBlend {blendOption=Linear}
    boneToModify {boneName=dyn_jacket_bk_r_02}
    constraintSetup [{targetBone {boneName=spine_04}}]
    constraintWeights [1]
   }
   n6 = "Constraint - dyn_jacket_bk_01 (P)"(ComponentPose=@n5, Alpha=@n1, ConstraintWeights_0=1) {
    alphaBoolBlend {blendOption=Linear}
    boneToModify {boneName=dyn_jacket_bk_01}
    constraintSetup [{targetBone {boneName=spine_04}}]
    constraintWeights [1]
   }
   n7 = "Constraint - dyn_jacket_bk_02 (P)"(ComponentPose=@n6, Alpha=@n1, ConstraintWeights_0=1) {
    alphaBoolBlend {blendOption=Linear}
    boneToModify {boneName=dyn_jacket_bk_02}
    constraintSetup [{targetBone {boneName=spine_04}}]
    constraintWeights [1]
   }
   n8 = "Constraint - dyn_jacket_l_01 (P)"(ComponentPose=@n7, Alpha=@n1, ConstraintWeights_0=1) {
    alphaBoolBlend {blendOption=Linear}
    boneToModify {boneName=dyn_jacket_l_01}
    constraintSetup [{targetBone {boneName=spine_04}}]
    constraintWeights [1]
   }
   "Constraint - dyn_jacket_r_01 (P)"(ComponentPose=@n8, Alpha=@n1, ConstraintWeights_0=1) {
    alphaBoolBlend {blendOption=Linear}
    boneToModify {boneName=dyn_jacket_r_01}
    constraintSetup [{targetBone {boneName=spine_04}}]
    constraintWeights [1]
   }
   RigidBody() {
    bClampLinearTranslationLimitToRefPose=true
    bOverrideWorldGravity=true
    overridePhysicsAsset=&/F_MED_Robot_Trouble_Dyn_Leg_l
    simulationSpace=BaseBoneSpace
    alphaBoolBlend {blendOption=Linear}
    baseBoneRef {boneName=thigh_l}
    componentAppliedLinearAccClamp=<3000 3000 3000>
    componentLinearAccScale=<0.6 0.6 0.6>
    componentLinearVelScale=<0.6 0.6 0.6>
    overrideWorldGravity {z=-980}
   }
   n9 = RigidBody() {
    bClampLinearTranslationLimitToRefPose=true
    bOverrideWorldGravity=true
    overridePhysicsAsset=&/F_MED_Robot_Trouble_Dyn_Collar
    simulationSpace=BaseBoneSpace
    alphaBoolBlend {blendOption=Linear}
    baseBoneRef {boneName=spine_02}
    componentAppliedLinearAccClamp=<3000 3000 3000>
    componentLinearAccScale=<2 1 1>
    componentLinearVelScale=<2 1 1>
    overrideWorldGravity {z=-980}
   }
   n10 = Break Fort Anim Input Skydiving External Force(FortAnimInput_SkydivingExternalForce=Get SkydivingExternalForceInput()) -> (
    bUseSkydivingVectorForce=~
    SkydivingVectorForce=~
    HeadToPelvisDirection=~
    FloatingMultiplier=~
    FloatingAdditive=~
    DivingMultiplier=~
    DivingAdditive=~
    ParachutingMultiplier=~
    ParachutingAdditive=~
   )
   n11 = RigidBody(ComponentPose=@n9, ExternalForce=@n10.SkydivingVectorForce) {
    bClampLinearTranslationLimitToRefPose=true
    bOverrideWorldGravity=true
    overridePhysicsAsset=&/F_MED_Robot_Trouble_Dyn_Jacket
    simulationSpace=BaseBoneSpace
    alphaBoolBlend {blendOption=Linear}
    baseBoneRef {boneName=spine_01}
    componentAppliedLinearAccClamp=<3000 3000 3000>
    componentLinearAccScale=<0.35 0.2 0.2>
    componentLinearVelScale=<0.35 0.2 0.2>
    overrideWorldGravity {z=-980}
   }
   n12 = RigidBody(ComponentPose=@n11, ExternalForce=@n10.SkydivingVectorForce) {
    bClampLinearTranslationLimitToRefPose=true
    bOverrideWorldGravity=true
    overridePhysicsAsset=&/F_MED_Robot_Trouble_Dyn_Straps
    simulationSpace=BaseBoneSpace
    alphaBoolBlend {blendOption=Linear}
    baseBoneRef {boneName=spine_04}
    componentAppliedLinearAccClamp=<3000 3000 3000>
    componentLinearAccScale=<0.12 0.4 0.4>
    componentLinearVelScale=<0.12 0.4 0.4>
    overrideWorldGravity {z=-980}
   }
   n13 = RigidBody(ComponentPose=@n12) {
    bClampLinearTranslationLimitToRefPose=true
    bOverrideWorldGravity=true
    overridePhysicsAsset=&F_MED_Robot_Trouble_Dyn_Leg_l
    simulationSpace=BaseBoneSpace
    alphaBoolBlend {blendOption=Linear}
    baseBoneRef {boneName=thigh_l}
    componentAppliedLinearAccClamp=<3000 3000 3000>
    componentLinearAccScale=<0.6 0.6 0.6>
    componentLinearVelScale=<0.6 0.6 0.6>
    overrideWorldGravity {z=-980}
   }
   n14 = RigidBody(ComponentPose=@n13) {
    bClampLinearTranslationLimitToRefPose=true
    bOverrideWorldGravity=true
    overridePhysicsAsset=&/F_MED_Robot_Trouble_Dyn_Arm_l
    simulationSpace=BaseBoneSpace
    alphaBoolBlend {blendOption=Linear}
    baseBoneRef {boneName=spine_01}
    componentAppliedLinearAccClamp=<3000 3000 3000>
    componentLinearAccScale=<0.5 0.5 0.5>
    componentLinearVelScale=<0.5 0.5 0.5>
    overrideWorldGravity {z=-980}
   }
   RigidBody(ComponentPose=@n14) {
    bClampLinearTranslationLimitToRefPose=true
    bOverrideWorldGravity=true
    overridePhysicsAsset=&/F_MED_Robot_Trouble_Dyn_Arm_r
    simulationSpace=BaseBoneSpace
    alphaBoolBlend {blendOption=Linear}
    baseBoneRef {boneName=spine_01}
    componentAppliedLinearAccClamp=<3000 3000 3000>
    componentLinearAccScale=<0.5 0.5 0.5>
    componentLinearVelScale=<0.5 0.5 0.5>
    overrideWorldGravity {z=-980}
   }
   n15 = "Transform (Modify) Bone - Bone: dyn_jacket_r_01_MSTR"(
    ComponentPose=Local To Component(LocalPose=Use cached pose 'Pre-Sim'())
    Alpha=Get jumping alpha() -> (jumping alpha:double)
    Translation="0.000,-0.000,0.890"
    Rotation="-33.887,0.361,-5.186"
    Scale="1,1,1"
   ) {
    rotationMode=BMM_Replace
    rotationSpace=BCS_BoneSpace
    translationMode=BMM_Replace
    translationSpace=BCS_BoneSpace
    alphaBoolBlend {blendOption=Linear}
    alphaScaleBiasClamp {interpSpeedDecreasing=1, interpSpeedIncreasing=1}
    boneToModify {boneName=dyn_jacket_r_01_MSTR}
    rotation=<rot -33.887 0.361 -5.186>
    translation {z=0.89}
   }
   n16 = RigidBody(ComponentPose=@n15) {
    bClampLinearTranslationLimitToRefPose=true
    bOverrideWorldGravity=true
    overridePhysicsAsset=&/F_MED_Robot_Trouble_Dyn_Collar_2
    simulationSpace=BaseBoneSpace
    alphaBoolBlend {blendOption=Linear}
    baseBoneRef {boneName=spine_05}
    componentAppliedLinearAccClamp=<3000 3000 3000>
    componentLinearAccScale=<0.3 0.3 0.3>
    componentLinearVelScale=<0.3 0.3 0.3>
    overrideWorldGravity {z=-980}
   }
   n17 = RigidBody(ComponentPose=@n16, ExternalForce=@n10.SkydivingVectorForce) {
    bClampLinearTranslationLimitToRefPose=true
    bOverrideWorldGravity=true
    overridePhysicsAsset=&/F_MED_Robot_Trouble_Dyn_Jacket1
    simulationSpace=BaseBoneSpace
    alphaBoolBlend {blendOption=Linear}
    baseBoneRef {boneName=spine_04}
    componentAppliedLinearAccClamp=<2000 2000 2000>
    componentLinearAccScale=<0.35 0.25 0.25>
    componentLinearVelScale=<0.3 0.2 0.2>
    overrideWorldGravity {z=-980}
   }
   n18 = RigidBody(ComponentPose=@n17, ExternalForce=@n10.SkydivingVectorForce) {
    bClampLinearTranslationLimitToRefPose=true
    bOverrideWorldGravity=true
    overridePhysicsAsset=&/F_MED_Robot_Trouble_Dyn_Straps_2
    simulationSpace=BaseBoneSpace
    alphaBoolBlend {blendOption=Linear}
    baseBoneRef {boneName=spine_05}
    componentAppliedLinearAccClamp=<3000 3000 3000>
    componentLinearAccScale=<0.12 0.4 0.4>
    componentLinearVelScale=<0.12 0.4 0.4>
    overrideWorldGravity {z=-980}
   }
   n19 = RigidBody(ComponentPose=@n18) {
    bClampLinearTranslationLimitToRefPose=true
    bOverrideWorldGravity=true
    overridePhysicsAsset=&F_MED_Robot_Trouble_Dyn_Arm_l
    simulationSpace=BaseBoneSpace
    alphaBoolBlend {blendOption=Linear}
    baseBoneRef {boneName=clavicle_l}
    componentAppliedLinearAccClamp=<3000 3000 3000>
    componentLinearAccScale=<0.2 0.2 0.2>
    componentLinearVelScale=<0.2 0.2 0.2>
    overrideWorldGravity {z=-980}
   }
   n20 = RigidBody(ComponentPose=@n19) {
    bClampLinearTranslationLimitToRefPose=true
    bOverrideWorldGravity=true
    overridePhysicsAsset=&F_MED_Robot_Trouble_Dyn_Arm_r
    simulationSpace=BaseBoneSpace
    alphaBoolBlend {blendOption=Linear}
    baseBoneRef {boneName=clavicle_r}
    componentAppliedLinearAccClamp=<3000 3000 3000>
    componentLinearAccScale=<0.25 0.25 0.25>
    componentLinearVelScale=<0.25 0.25 0.25>
    overrideWorldGravity {z=-980}
   }
   n21 = Anim Dynamics - Bone: dyn_thigh_pouch_l(ComponentPose=@n20, Alpha=1, ExternalForce="0,0,0") {
    angularDampingOverride=0.9995
    bUseGravityOverride=true
    simulationSpace=RootRelative
    alphaBoolBlend {blendOption=Linear}
    boundBone {boneName=dyn_thigh_pouch_l}
    componentAppliedLinearAccClamp=<3000 3000 3000>
    componentLinearAccScale=<0.25 0.25 0.25>
    componentLinearVelScale=<0.3 0.3 0.3>
    gravityOverride {z=-980}
    physicsBodyDefinitions [{
      sphereCollisionRadius=0
      boundBone {boneName=dyn_thigh_pouch_l}
      boxExtents=<30 10 10>
      constraintSetup {angularLimitsMax=<0 5 15>, angularLimitsMin=<0 -15 -25>, angularTarget=<0 0 0>}
      localJointOffset=<2 0 0>
     }]
    relativeSpaceBone {boneName=thigh_l}
   }
   n22 = Anim Dynamics - Bone: dyn_calf_pouch_l(ComponentPose=@n21, Alpha=0.45, ExternalForce="0,0,0") {
    angularDampingOverride=0.9995
    angularSpringConstant=1000
    bAngularSpring=true
    bUseGravityOverride=true
    simulationSpace=RootRelative
    alphaBoolBlend {blendOption=Linear}
    boundBone {boneName=dyn_calf_pouch_l}
    componentAppliedLinearAccClamp=<3000 3000 3000>
    componentLinearAccScale=<0.1 0.1 0.1>
    componentLinearVelScale=<0.2 0.2 0.2>
    gravityOverride {z=-980}
    physicsBodyDefinitions [{
      sphereCollisionRadius=0
      boundBone {boneName=dyn_calf_pouch_l}
      boxExtents=<40 12 12>
      constraintSetup {angularLimitsMax=<0 5 8>, angularLimitsMin=<0 -5 -8>, angularTarget=<0 0 0>}
      localJointOffset=<2.5 0 0>
     }]
    relativeSpaceBone {boneName=calf_l}
   }
   n23 = Anim Dynamics - Bone: dyn_satchel(ComponentPose=@n22, Alpha=1, ExternalForce="0,0,0") {
    angularDampingOverride=0.9999
    angularSpringConstant=150
    bAngularSpring=true
    bUseGravityOverride=true
    simulationSpace=RootRelative
    alphaBoolBlend {blendOption=Linear}
    boundBone {boneName=dyn_satchel}
    componentAppliedLinearAccClamp=<3000 3000 3000>
    componentLinearAccScale=<0.3 0.3 0.4>
    componentLinearVelScale=<0.3 0.3 0.3>
    gravityOverride {z=-980}
    physicsBodyDefinitions [{
      sphereCollisionRadius=0
      boundBone {boneName=dyn_satchel}
      boxExtents=<40 20 10>
      constraintSetup {angularTargetAxis=AxisY, angularLimitsMax=<0 3 2>, angularLimitsMin=<0 -25 -5>, angularTarget=<0 -20 0>}
      localJointOffset=<3.1 0 0>
     }]
    relativeSpaceBone {boneName=pelvis}
   }
   n24 = "Constraint - dyn_jacket_bk_01 (PP)"!(ComponentPose=@n23, Alpha=@n1, ConstraintWeights_0=0.4483, ConstraintWeights_1:float=0.5517) -> (Pose) {
    alphaBoolBlend {blendOption=Linear}
    boneToModify {boneName=dyn_jacket_bk_01}
    constraintSetup [{targetBone {boneName=spine_04}}, {targetBone {boneName=spine_05}}]
    constraintWeights [1,1]
   }
   n25 = "Constraint - dyn_jacket_bk_l_01 (PP)"!(ComponentPose=@n24, Alpha=@n1, ConstraintWeights_0=0.4483, ConstraintWeights_1:float=0.5517) -> (Pose) {
    alphaBoolBlend {blendOption=Linear}
    boneToModify {boneName=dyn_jacket_bk_l_01}
    constraintSetup [{targetBone {boneName=spine_04}}, {targetBone {boneName=spine_05}}]
    constraintWeights [1,1]
   }
   n26 = "Constraint - dyn_jacket_bk_r_01 (PP)"!(ComponentPose=@n25, Alpha=@n1, ConstraintWeights_0=0.4483, ConstraintWeights_1:float=0.5517) -> (Pose) {
    alphaBoolBlend {blendOption=Linear}
    boneToModify {boneName=dyn_jacket_bk_r_01}
    constraintSetup [{targetBone {boneName=spine_04}}, {targetBone {boneName=spine_05}}]
    constraintWeights [1,1]
   }
   n27 = "Constraint - dyn_jacket_l_01 (PP)"!(ComponentPose=@n26, Alpha=@n1, ConstraintWeights_0=0.4483, ConstraintWeights_1:float=0.5517) -> (Pose) {
    alphaBoolBlend {blendOption=Linear}
    boneToModify {boneName=dyn_jacket_l_01}
    constraintSetup [{targetBone {boneName=spine_04}}, {targetBone {boneName=spine_05}}]
    constraintWeights [1,1]
   }
   n28 = "Constraint - dyn_jacket_bk_02 (PP)"!(ComponentPose=@n27, Alpha=@n1, ConstraintWeights_0=0.4483, ConstraintWeights_1:float=0.5517) -> (Pose) {
    alphaBoolBlend {blendOption=Linear}
    boneToModify {boneName=dyn_jacket_bk_02}
    constraintSetup [{targetBone {boneName=spine_04}}, {targetBone {boneName=spine_05}}]
    constraintWeights [1,1]
   }
   n29 = "Constraint - dyn_jacket_r_01 (PP)"!(ComponentPose=@n28, Alpha=@n1, ConstraintWeights_0=0.4483, ConstraintWeights_1:float=0.5517) -> (Pose) {
    alphaBoolBlend {blendOption=Linear}
    boneToModify {boneName=dyn_jacket_r_01}
    constraintSetup [{targetBone {boneName=spine_04}}, {targetBone {boneName=spine_05}}]
    constraintWeights [1,1]
   }
   n30 = "Constraint - dyn_jacket_bk_l_02 (PP)"!(ComponentPose=@n29, Alpha=@n1, ConstraintWeights_0=0.4483, ConstraintWeights_1:float=0.5517) -> (Pose) {
    alphaBoolBlend {blendOption=Linear}
    boneToModify {boneName=dyn_jacket_bk_l_02}
    constraintSetup [{targetBone {boneName=spine_04}}, {targetBone {boneName=spine_05}}]
    constraintWeights [1,1]
   }
   n31 = "Constraint - dyn_jacket_bk_r_02 (PP)"!(ComponentPose=@n30, Alpha=@n1, ConstraintWeights_0=0.4483, ConstraintWeights_1:float=0.5517) -> (Pose) {
    alphaBoolBlend {blendOption=Linear}
    boneToModify {boneName=dyn_jacket_bk_r_02}
    constraintSetup [{targetBone {boneName=spine_04}}, {targetBone {boneName=spine_05}}]
    constraintWeights [1,1]
   }
   n32 = "Transform (Modify) Bone - Bone: dyn_jacket_bk_01"!(ComponentPose=@n31, bAlphaBoolEnabled:bool=True, Translation="0,0,2.6", Rotation="-5,0,0", Scale="1,1,1") -> (Pose) {
    alphaInputType=Bool
    bAlphaBoolEnabled=false
    rotationMode=BMM_Additive
    rotationSpace=BCS_BoneSpace
    translationMode=BMM_Additive
    translationSpace=BCS_BoneSpace
    alphaBoolBlend {blendOption=Linear}
    boneToModify {boneName=dyn_jacket_bk_01}
    rotation {pitch=-5}
    translation {z=2.6}
   }
   n33 = "Transform (Modify) Bone - Bone: dyn_jacket_bk_l_01"!(ComponentPose=@n32, bAlphaBoolEnabled:bool=True, Translation="0,0,2.5", Rotation="-5,0,0", Scale="1,1,1") -> (Pose) {
    alphaInputType=Bool
    bAlphaBoolEnabled=false
    rotationMode=BMM_Additive
    rotationSpace=BCS_BoneSpace
    translationMode=BMM_Additive
    translationSpace=BCS_BoneSpace
    alphaBoolBlend {blendOption=Linear}
    boneToModify {boneName=dyn_jacket_bk_l_01}
    rotation {pitch=-5}
    translation {z=2.5}
   }
   n34 = "Transform (Modify) Bone - Bone: dyn_jacket_bk_r_01"!(ComponentPose=@n33, bAlphaBoolEnabled:bool=True, Translation="0,0,-2.5", Rotation="-5,0,0", Scale="1,1,1") -> (Pose) {
    alphaInputType=Bool
    bAlphaBoolEnabled=false
    rotationMode=BMM_Additive
    rotationSpace=BCS_BoneSpace
    translationMode=BMM_Additive
    translationSpace=BCS_BoneSpace
    alphaBoolBlend {blendOption=Linear}
    boneToModify {boneName=dyn_jacket_bk_r_01}
    rotation {pitch=-5}
    translation {z=-2.5}
   }
   n35 = Get Crouch Alpha() -> (Crouch Alpha:double)
   n36 = "Constraint - dyn_collar_back (P)"(ComponentPose=@n34, Alpha=@n35, ConstraintWeights_0=1) {
    alphaBoolBlend {blendOption=Linear}
    alphaScaleBiasClamp {bInterpResult=true}
    boneToModify {boneName=dyn_collar_back}
    constraintSetup [{targetBone {boneName=neck_01}}]
    constraintWeights [1]
   }
   n37 = float * float(A=@n35, B=0.8) {is=title K2Node_PromotableOperator}
   n38 = "Constraint - dyn_collar_l (P)"(ComponentPose=@n36, Alpha=@n37, ConstraintWeights_0=1) {
    alphaBoolBlend {blendOption=Linear}
    alphaScaleBiasClamp {bInterpResult=true}
    boneToModify {boneName=dyn_collar_l}
    constraintSetup [{targetBone {boneName=neck_01}}]
    constraintWeights [1]
   }
   n39 = "Constraint - dyn_collar_r (P)"(ComponentPose=@n38, Alpha=@n37, ConstraintWeights_0=1) {
    alphaBoolBlend {blendOption=Linear}
    alphaScaleBiasClamp {bInterpResult=true}
    boneToModify {boneName=dyn_collar_r}
    constraintSetup [{targetBone {boneName=neck_01}}]
    constraintWeights [1]
   }
   n40 = Get Crouching Neck Alpha() -> (Crouching Neck Alpha:double)
   n41 = "Transform (Modify) Bone - Bone: dyn_collar_back"(ComponentPose=@n39, Alpha=@n40, Translation="0,0,0", Rotation="0,-20,0", Scale="1,1,1") {
    rotationMode=BMM_Additive
    rotationSpace=BCS_BoneSpace
    translationMode=BMM_Additive
    translationSpace=BCS_BoneSpace
    alphaBoolBlend {blendOption=Linear}
    boneToModify {boneName=dyn_collar_back}
    rotation {yaw=-20}
   }
   n42 = "Transform (Modify) Bone - Bone: dyn_collar_l"(ComponentPose=@n41, Alpha=@n40, Translation="0,0,0", Rotation="5,-20,0", Scale="1,1,1") {
    rotationMode=BMM_Additive
    rotationSpace=BCS_BoneSpace
    translationMode=BMM_Additive
    translationSpace=BCS_BoneSpace
    alphaBoolBlend {blendOption=Linear}
    boneToModify {boneName=dyn_collar_l}
    rotation {pitch=5, yaw=-20}
   }
   n43 = "Transform (Modify) Bone - Bone: dyn_collar_r"(ComponentPose=@n42, Alpha=@n40, Translation="0,0,0", Rotation="-5,-20,0", Scale="1,1,1") {
    rotationMode=BMM_Additive
    rotationSpace=BCS_BoneSpace
    translationMode=BMM_Additive
    translationSpace=BCS_BoneSpace
    alphaBoolBlend {blendOption=Linear}
    boneToModify {boneName=dyn_collar_r}
    rotation {pitch=-5, yaw=-20}
   }
   n44 = "Constraint - dyn_jacket_bk_01 (PP)"(ComponentPose=Local To Component(LocalPose=Use cached pose 'Pre-Sim'()), Alpha=1, ConstraintWeights_0=0.4483, ConstraintWeights_1=0.5517) {
    alphaBoolBlend {blendOption=Linear}
    boneToModify {boneName=dyn_jacket_bk_01}
    constraintSetup [{targetBone {boneName=spine_04}}, {targetBone {boneName=spine_05}}]
    constraintWeights [1,1]
   }
   n45 = "Constraint - dyn_jacket_bk_l_01 (PP)"(ComponentPose=@n44, Alpha=1, ConstraintWeights_0=0.4483, ConstraintWeights_1=0.5517) {
    alphaBoolBlend {blendOption=Linear}
    boneToModify {boneName=dyn_jacket_bk_l_01}
    constraintSetup [{targetBone {boneName=spine_04}}, {targetBone {boneName=spine_05}}]
    constraintWeights [1,1]
   }
   n46 = "Constraint - dyn_jacket_bk_r_01 (PP)"(ComponentPose=@n45, Alpha=1, ConstraintWeights_0=0.4483, ConstraintWeights_1=0.5517) {
    alphaBoolBlend {blendOption=Linear}
    boneToModify {boneName=dyn_jacket_bk_r_01}
    constraintSetup [{targetBone {boneName=spine_04}}, {targetBone {boneName=spine_05}}]
    constraintWeights [1,1]
   }
   n47 = "Constraint - dyn_jacket_l_01 (PP)"(ComponentPose=@n46, Alpha=1, ConstraintWeights_0=0.4483, ConstraintWeights_1=0.5517) {
    alphaBoolBlend {blendOption=Linear}
    boneToModify {boneName=dyn_jacket_l_01}
    constraintSetup [{targetBone {boneName=spine_04}}, {targetBone {boneName=spine_05}}]
    constraintWeights [1,1]
   }
   n48 = "Constraint - dyn_jacket_fr_l_01 (PP)"!(ComponentPose=@n47, Alpha=1, ConstraintWeights_0=0.4483, ConstraintWeights_1:float=0.5517) -> (Pose) {
    alphaBoolBlend {blendOption=Linear}
    boneToModify {boneName=dyn_jacket_fr_l_01}
    constraintSetup [{targetBone {boneName=spine_04}}, {targetBone {boneName=spine_05}}]
    constraintWeights [1,1]
   }
   n49 = "Constraint - dyn_jacket_r_01 (PP)"(ComponentPose=@n48, Alpha=1, ConstraintWeights_0=0.4483, ConstraintWeights_1=0.5517) {
    alphaBoolBlend {blendOption=Linear}
    boneToModify {boneName=dyn_jacket_r_01}
    constraintSetup [{targetBone {boneName=spine_04}}, {targetBone {boneName=spine_05}}]
    constraintWeights [1,1]
   }
   n50 = "Constraint - dyn_jacket_fr_r_01 (PP)"!(ComponentPose=@n49, Alpha=1, ConstraintWeights_0=0.4483, ConstraintWeights_1:float=0.5517) -> (Pose) {
    alphaBoolBlend {blendOption=Linear}
    boneToModify {boneName=dyn_jacket_fr_r_01}
    constraintSetup [{targetBone {boneName=spine_04}}, {targetBone {boneName=spine_05}}]
    constraintWeights [1,1]
   }
   n51 = "Constraint - dyn_collar_back (P)"(ComponentPose=@n50, Alpha=1, ConstraintWeights_0=1) {
    alphaBoolBlend {blendOption=Linear}
    alphaScaleBiasClamp {bInterpResult=true}
    boneToModify {boneName=dyn_collar_back}
    constraintSetup [{targetBone {boneName=neck_01}}]
    constraintWeights [1]
   }
   n52 = "Constraint - dyn_collar_l (P)"(ComponentPose=@n51, Alpha=0.7, ConstraintWeights_0=1) {
    alphaBoolBlend {blendOption=Linear}
    alphaScaleBiasClamp {bInterpResult=true}
    boneToModify {boneName=dyn_collar_l}
    constraintSetup [{targetBone {boneName=neck_01}}]
    constraintWeights [1]
   }
   n53 = "Constraint - dyn_collar_r (P)"(ComponentPose=@n52, Alpha=0.7, ConstraintWeights_0=1) {
    alphaBoolBlend {blendOption=Linear}
    alphaScaleBiasClamp {bInterpResult=true}
    boneToModify {boneName=dyn_collar_r}
    constraintSetup [{targetBone {boneName=neck_01}}]
    constraintWeights [1]
   }
   n54 = Get Crouching Neck Alpha()
   n55 = "Transform (Modify) Bone - Bone: dyn_collar_back"(ComponentPose=@n53, Alpha=@n54, Translation="0,0,0", Rotation="0,-20,0", Scale="1,1,1") {
    rotationMode=BMM_Additive
    rotationSpace=BCS_BoneSpace
    translationMode=BMM_Additive
    translationSpace=BCS_BoneSpace
    alphaBoolBlend {blendOption=Linear}
    boneToModify {boneName=dyn_collar_back}
    rotation {yaw=-20}
   }
   n56 = "Transform (Modify) Bone - Bone: dyn_collar_l"(ComponentPose=@n55, Alpha=@n54, Translation="0,0,0", Rotation="0,-20,0", Scale="1,1,1") {
    rotationMode=BMM_Additive
    rotationSpace=BCS_BoneSpace
    translationMode=BMM_Additive
    translationSpace=BCS_BoneSpace
    alphaBoolBlend {blendOption=Linear}
    boneToModify {boneName=dyn_collar_l}
    rotation {yaw=-20}
   }
   n57 = "Transform (Modify) Bone - Bone: dyn_collar_r"(ComponentPose=@n56, Alpha=@n54, Translation="0,0,0", Rotation="0,-20,0", Scale="1,1,1") {
    rotationMode=BMM_Additive
    rotationSpace=BCS_BoneSpace
    translationMode=BMM_Additive
    translationSpace=BCS_BoneSpace
    alphaBoolBlend {blendOption=Linear}
    boneToModify {boneName=dyn_collar_r}
    rotation {yaw=-20}
   }
   n58 = "Transform (Modify) Bone - Bone: dyn_jacket_bk_01"(ComponentPose=@n57, bAlphaBoolEnabled=True, Translation="0,0,2.6", Rotation="-5,0,0", Scale="1,1,1") {
    alphaInputType=Bool
    bAlphaBoolEnabled=false
    rotationMode=BMM_Additive
    rotationSpace=BCS_BoneSpace
    translationMode=BMM_Additive
    translationSpace=BCS_BoneSpace
    alphaBoolBlend {blendOption=Linear}
    boneToModify {boneName=dyn_jacket_bk_01}
    rotation {pitch=-5}
    translation {z=2.6}
   }
   n59 = "Transform (Modify) Bone - Bone: dyn_jacket_bk_l_01"(ComponentPose=@n58, bAlphaBoolEnabled=True, Translation="0,-0.25,3", Rotation="0,0,0", Scale="1,1,1") {
    alphaInputType=Bool
    bAlphaBoolEnabled=false
    rotationMode=BMM_Additive
    rotationSpace=BCS_BoneSpace
    translationMode=BMM_Additive
    translationSpace=BCS_BoneSpace
    alphaBoolBlend {blendOption=Linear}
    boneToModify {boneName=dyn_jacket_bk_l_01}
    translation {y=-0.25, z=3}
   }
   n60 = "Transform (Modify) Bone - Bone: dyn_jacket_bk_r_01"(ComponentPose=@n59, bAlphaBoolEnabled=True, Translation="0,0,-2.5", Rotation="-5,0,0", Scale="1,1,1") {
    alphaInputType=Bool
    bAlphaBoolEnabled=false
    rotationMode=BMM_Additive
    rotationSpace=BCS_BoneSpace
    translationMode=BMM_Additive
    translationSpace=BCS_BoneSpace
    alphaBoolBlend {blendOption=Linear}
    boneToModify {boneName=dyn_jacket_bk_r_01}
    rotation {pitch=-5}
    translation {z=-2.5}
   }
   Output Pose(Result=Blend Poses by bool(BlendPose_0=Component To Local(ComponentPose=@n43), BlendPose_1=Component To Local(ComponentPose=@n60), bActiveValue=Get bRigidBodyEnabled()) {blendType=Linear}) {layerGroup=None}
   n61 = "Transform (Modify) Bone - Bone: dyn_collar_r_MSTR"(
    ComponentPose=Local To Component(LocalPose=Copy Pose From Mesh() {bUseAttachedParent=true})
    Alpha=Get arm_up_r() -> (arm_up_r:double)
    Translation="0.147,0.000,-3.300"
    Rotation="-10.570,0.000,0.000"
    Scale="1,1,1"
   ) {
    rotationMode=BMM_Additive
    rotationSpace=BCS_BoneSpace
    translationMode=BMM_Additive
    translationSpace=BCS_BoneSpace
    alphaBoolBlend {blendOption=Linear}
    alphaScaleBiasClamp {interpSpeedDecreasing=1, interpSpeedIncreasing=1}
    boneToModify {boneName=dyn_collar_r_MSTR}
    rotation {pitch=-10.57}
    translation {x=0.147, z=-3.3}
   }
   n62 = "Transform (Modify) Bone - Bone: dyn_collar_l_MSTR"(ComponentPose=@n61, Alpha=Get arm_up_l() -> (arm_up_l:double), Translation="0.651,0.000,2", Rotation="17.901,0.000,0.000", Scale="1,1,1") {
    rotationMode=BMM_Additive
    rotationSpace=BCS_BoneSpace
    translationMode=BMM_Additive
    translationSpace=BCS_BoneSpace
    alphaBoolBlend {blendOption=Linear}
    alphaScaleBiasClamp {interpSpeedDecreasing=1, interpSpeedIncreasing=1}
    boneToModify {boneName=dyn_collar_l_MSTR}
    rotation {pitch=17.901}
    translation {x=0.651, z=2}
   }
   n63 = Look At - Bone: dyn_piston_elbow_fr_base_l(ComponentPose=@n62, Alpha=1) {
    alphaBoolBlend {blendOption=Linear}
    boneToModify {boneName=dyn_piston_elbow_fr_base_l}
    lookAtLocation {x=0}
    lookAtTarget {boneReference {boneName=dyn_piston_elbow_fr_end_l}}
    lookAt_Axis {axis=<1 0>}
   }
   n64 = Look At - Bone: dyn_piston_elbow_fr_end_l(ComponentPose=@n63, Alpha=1) {
    alphaBoolBlend {blendOption=Linear}
    boneToModify {boneName=dyn_piston_elbow_fr_end_l}
    lookAtLocation {x=0}
    lookAtTarget {boneReference {boneName=dyn_piston_elbow_fr_base_l}}
    lookAt_Axis {axis=<-1 0>}
   }
   n65 = Look At - Bone: dyn_piston_elbow_bk_base_l(ComponentPose=@n64, Alpha=1) {
    alphaBoolBlend {blendOption=Linear}
    boneToModify {boneName=dyn_piston_elbow_bk_base_l}
    lookAtLocation {x=0}
    lookAtTarget {boneReference {boneName=dyn_piston_elbow_bk_end_l}}
    lookAt_Axis {axis=<1 0>}
   }
   n66 = Look At - Bone: dyn_piston_elbow_bk_end_l(ComponentPose=@n65, Alpha=1) {
    alphaBoolBlend {blendOption=Linear}
    boneToModify {boneName=dyn_piston_elbow_bk_end_l}
    lookAtLocation {x=0}
    lookAtTarget {boneReference {boneName=dyn_piston_elbow_bk_base_l}}
    lookAt_Axis {axis=<-1 0>}
   }
   n67 = Look At - Bone: dyn_piston_elbow_fr_base_r(ComponentPose=@n66, Alpha=1) {
    alphaBoolBlend {blendOption=Linear}
    boneToModify {boneName=dyn_piston_elbow_fr_base_r}
    lookAtLocation {x=0}
    lookAtTarget {boneReference {boneName=dyn_piston_elbow_fr_end_r}}
    lookAt_Axis {axis=<-1 0>}
   }
   n68 = Look At - Bone: dyn_piston_elbow_fr_end_r(ComponentPose=@n67, Alpha=1) {
    alphaBoolBlend {blendOption=Linear}
    boneToModify {boneName=dyn_piston_elbow_fr_end_r}
    lookAtLocation {x=0}
    lookAtTarget {boneReference {boneName=dyn_piston_elbow_fr_base_r}}
    lookAt_Axis {axis=<1 0>}
   }
   n69 = Look At - Bone: dyn_piston_elbow_bk_base_r(ComponentPose=@n68, Alpha=1) {
    alphaBoolBlend {blendOption=Linear}
    boneToModify {boneName=dyn_piston_elbow_bk_base_r}
    lookAtLocation {x=0}
    lookAtTarget {boneReference {boneName=dyn_piston_elbow_bk_end_r}}
    lookAt_Axis {axis=<-1 0>}
   }
   n70 = Look At - Bone: dyn_piston_elbow_bk_end_r(ComponentPose=@n69, Alpha=1) {
    alphaBoolBlend {blendOption=Linear}
    boneToModify {boneName=dyn_piston_elbow_bk_end_r}
    lookAtLocation {x=0}
    lookAtTarget {boneReference {boneName=dyn_piston_elbow_bk_base_r}}
    lookAt_Axis {axis=<1 0>}
   }
   n71 = "dyn_elbow_joint_l.translateX = curve(lowerarm_l.rotateZ) - Bone Driven Controller"!(ComponentPose:Component Space Pose Link Structure=@n70, Alpha:float=1) -> (Pose:Component Space Pose Link Structure) {
    is=title AnimGraphNode_BoneDrivenController
    bAffectTargetTranslationX=true
    drivingCurve=&/F_MED_Robot_Trouble_elbow_joint_curve_l
    sourceComponent=RotationZ
    alphaBoolBlend {blendOption=Linear}
    sourceBone {boneName=lowerarm_l}
    targetBone {boneName=dyn_elbow_joint_l}
   }
   n72 = "dyn_elbow_joint_r.translateX = curve(lowerarm_r.rotateZ) - Bone Driven Controller"!(ComponentPose:Component Space Pose Link Structure=@n71, Alpha:float=1) -> (Pose:Component Space Pose Link Structure) {
    is=title AnimGraphNode_BoneDrivenController
    bAffectTargetTranslationX=true
    drivingCurve=&/F_MED_Robot_Trouble_elbow_joint_r_curve
    sourceComponent=RotationZ
    alphaBoolBlend {blendOption=Linear}
    sourceBone {boneName=lowerarm_r}
    targetBone {boneName=dyn_elbow_joint_r}
   }
   n73 = Look At - Bone: dyn_piston_knee_base_l(ComponentPose=@n72, Alpha=1) {
    alphaBoolBlend {blendOption=Linear}
    boneToModify {boneName=dyn_piston_knee_base_l}
    lookAtLocation {x=0}
    lookAtTarget {boneReference {boneName=dyn_piston_knee_end_l}}
    lookAt_Axis {axis=<1 0>}
   }
   n74 = Look At - Bone: dyn_piston_knee_end_l(ComponentPose=@n73, Alpha=1) {
    alphaBoolBlend {blendOption=Linear}
    boneToModify {boneName=dyn_piston_knee_end_l}
    lookAtLocation {x=0}
    lookAtTarget {boneReference {boneName=dyn_piston_knee_base_l}}
    lookAt_Axis {axis=<-1 0>}
   }
   n75 = Look At - Bone: dyn_piston_knee_base_r(ComponentPose=@n74, Alpha=1) {
    alphaBoolBlend {blendOption=Linear}
    boneToModify {boneName=dyn_piston_knee_base_r}
    lookAtLocation {x=0}
    lookAtTarget {boneReference {boneName=dyn_piston_knee_end_r}}
    lookAt_Axis {axis=<-1 0>}
   }
   n76 = Look At - Bone: dyn_piston_knee_end_r(ComponentPose=@n75, Alpha=1) {
    alphaBoolBlend {blendOption=Linear}
    boneToModify {boneName=dyn_piston_knee_end_r}
    lookAtLocation {x=0}
    lookAtTarget {boneReference {boneName=dyn_piston_knee_base_r}}
    lookAt_Axis {axis=<1 0>}
   }
   n77 = "Constraint - dyn_shoulder_strap_bk_l_MSTR (PPP)"!(ComponentPose=@n76, Alpha=1, ConstraintWeights_0=0.0888, ConstraintWeights_1:float=0.0805, ConstraintWeights_2:float=0.8307) -> (Pose) {
    alphaBoolBlend {blendOption=Linear}
    boneToModify {boneName=dyn_shoulder_strap_bk_l_MSTR}
    constraintSetup [{targetBone {boneName=upperarm_twist_01_l}}, {targetBone {boneName=spine_05}}, {targetBone {boneName=clavicle_l}}]
    constraintWeights [1,1,1]
   }
   n78 = "Constraint - dyn_shoulder_strap_bk_r_MSTR (PP)"!(ComponentPose=@n77, Alpha=1, ConstraintWeights_0=0.8998, ConstraintWeights_1:float=0.1002) -> (Pose) {
    alphaBoolBlend {blendOption=Linear}
    boneToModify {boneName=dyn_shoulder_strap_bk_r_MSTR}
    constraintSetup [{targetBone {boneName=clavicle_r}}, {targetBone {boneName=spine_05}}]
    constraintWeights [1,1]
   }
   n79 = "Constraint - dyn_shoulder_strap_fr_l_MSTR (PPPPP)"!(
    ComponentPose=@n78
    Alpha=1
    ConstraintWeights_0=0.0115
    ConstraintWeights_1:float=0
    ConstraintWeights_2:float=0.2807
    ConstraintWeights_3:float=0.3896
    ConstraintWeights_4:float=0.3182
   ) -> (Pose) {
    alphaBoolBlend {blendOption=Linear}
    boneToModify {boneName=dyn_shoulder_strap_fr_l_MSTR}
    constraintSetup [
     {targetBone {boneName=clavicle_r}}
     {targetBone {boneName=upperarm_twist_01_l}}
     {targetBone {boneName=spine_04}}
     {targetBone {boneName=spine_05}}
     {targetBone {boneName=clavicle_l}}
    ]
    constraintWeights [1,1,1,1,1]
   }
   n80 = "Constraint - dyn_shoulder_strap_fr_r_MSTR (PPP)"!(ComponentPose=@n79, Alpha=1, ConstraintWeights_0=0.3281, ConstraintWeights_1:float=0.2946, ConstraintWeights_2:float=0.3772) -> (Pose) {
    alphaBoolBlend {blendOption=Linear}
    boneToModify {boneName=dyn_shoulder_strap_fr_r_MSTR}
    constraintSetup [{targetBone {boneName=clavicle_r}}, {targetBone {boneName=spine_04}}, {targetBone {boneName=spine_05}}]
    constraintWeights [1,1,1]
   }
   n81 = "Constraint - dyn_sleeve_piston_l (PP)"!(ComponentPose=@n80, Alpha=1, ConstraintWeights_0=1, ConstraintWeights_1:float=1) -> (Pose) {
    alphaBoolBlend {blendOption=Linear}
    boneToModify {boneName=dyn_sleeve_piston_l}
    constraintSetup [{targetBone {boneName=dyn_piston_elbow_fr_end_l}}, {targetBone {boneName=dyn_piston_elbow_fr_base_l}}]
    constraintWeights [1,1]
   }
   n82 = "Constraint - dyn_ball_l_alt (R)"(ComponentPose=@n81, Alpha=1, ConstraintWeights_0=1) {
    alphaBoolBlend {blendOption=Linear}
    boneToModify {boneName=dyn_ball_l_alt}
    constraintSetup [{transformType=Rotation, targetBone {boneName=ball_l}}]
    constraintWeights [1]
   }
   n83 = "Constraint - dyn_ball_r_alt (R)"(ComponentPose=@n82, Alpha=1, ConstraintWeights_0=1) {
    alphaBoolBlend {blendOption=Linear}
    boneToModify {boneName=dyn_ball_r_alt}
    constraintSetup [{transformType=Rotation, targetBone {boneName=ball_r}}]
    constraintWeights [1]
   }
   n84 = "Constraint - thigh_ball_pivot_l (P)"(ComponentPose=@n83, Alpha=1, ConstraintWeights_0=1) {
    alphaBoolBlend {blendOption=Linear}
    boneToModify {boneName=thigh_ball_pivot_l}
    constraintSetup [{targetBone {boneName=thigh_l}}]
    constraintWeights [1]
   }
   n85 = "Constraint - thigh_ball_pivot_r (P)"(ComponentPose=@n84, Alpha=1, ConstraintWeights_0=1) {
    alphaBoolBlend {blendOption=Linear}
    boneToModify {boneName=thigh_ball_pivot_r}
    constraintSetup [{targetBone {boneName=thigh_r}}]
    constraintWeights [1]
   }
   n86 = "Constraint - thigh_ball_pivot_l (R)"(ComponentPose=@n85, Alpha=1, ConstraintWeights_0=1) {
    alphaBoolBlend {blendOption=Linear}
    boneToModify {boneName=thigh_ball_pivot_l}
    constraintSetup [{transformType=Rotation, targetBone {boneName=pelvis}}]
    constraintWeights [1]
   }
   n87 = "Constraint - thigh_ball_pivot_r (R)"(ComponentPose=@n86, Alpha=1, ConstraintWeights_0=1) {
    alphaBoolBlend {blendOption=Linear}
    boneToModify {boneName=thigh_ball_pivot_r}
    constraintSetup [{transformType=Rotation, targetBone {boneName=pelvis}}]
    constraintWeights [1]
   }
   Save cached pose 'Pre-Sim'(Pose=Component To Local(ComponentPose=@n87)) {cacheName=Pre-Sim}
  ]}
 EventGraph {nodes [
   Event BlueprintUpdateAnimation()
   > n1 = Sequence!(execute) -> (then_0, then_1, then_2:Exec)
   > n2 = Branch(Condition=Get bIsJumping())
   > n3 = Set jumping alpha(jumping alpha:double=1) -> (Output_Get:double)
   Do Once(execute=@n2.else, Reset=@n3.then)
   > Delay(Duration=0.4)
   > Set jumping alpha(jumping alpha=0)
   n4 = GetOwnerSkeletalMeshComponent()
   > Is Valid(InputObject=@n4)
   n5 = BreakTransform(InTransform=GetDeltaTransformFromRefPose(self=@n4, BoneName=upperarm_r, BaseName=spine_05))
   > Set arm_up_r(arm_up_r:double=MapRangeClamped(Value=BreakRotator(InRot=@n5.Rotation).Yaw, InRangeA=80, InRangeB=138)) -> (Output_Get:double)
   n6 = GetOwnerSkeletalMeshComponent()
   > Is Valid(InputObject=@n6)
   n7 = BreakTransform(InTransform=GetDeltaTransformFromRefPose(self=@n6, BoneName=upperarm_l, BaseName=spine_05))
   > Set arm_up_l(arm_up_l:double=MapRangeClamped(Value=BreakRotator(InRot=@n7.Rotation).Yaw, InRangeA=50, InRangeB=105)) -> (Output_Get:double)
   n8 = Branch(execute=@n1.then_1, Condition=Get bIsDBNO())
   > Set DBNO Alpha(DBNO Alpha:double=0.75) -> (Output_Get:double)
   Set DBNO Alpha(execute=@n8.else, DBNO Alpha=0)
   > n9 = Branch(Condition=Get bIsCrouching())
   > Set Crouch Alpha(Crouch Alpha:double=1) -> (Output_Get:double)
   n10 = BreakTransform(InTransform=GetDeltaTransformFromRefPose(self=GetOwnerSkeletalMeshComponent(), BoneName=neck_01, BaseName=spine_05))
   > Set Crouching Neck Alpha(Crouching Neck Alpha:double=MapRangeClamped(Value=BreakRotator(InRot=@n10.Rotation).Yaw, InRangeA=0, InRangeB=-25)) -> (Output_Get:double)
   Set Crouch Alpha(execute=@n9.else, Crouch Alpha=0)
   n11 = Cast To FortPlayerAnimInstance(execute=@n1.then_2, Object=GetAnimInstance(self=GetOwnerSkeletalMeshComponent()))
   > Set Back X Rot Alpha(Back X Rot Alpha:double=0) -> (Output_Get:double)
   > n12 = Switch on EFortWeaponCoreAnimation!(execute, Selection:EFortWeaponCoreAnimation Enum=Get WeaponCoreAnim(self=@n11)) -> (
    Melee:Exec
    Pistol:Exec
    Shotgun:Exec
    PaperBlueprint:Exec
    Rifle:Exec
    MeleeOneHand:Exec
    MachinePistol:Exec
    RocketLauncher:Exec
    GrenadeLauncher:Exec
    GoingCommando:Exec
    AssaultRifle:Exec
    TacticalShotgun:Exec
    SniperRifle:Exec
    TrapPlacement:Exec
    ShoulderLauncher:Exec
    AbilityDecoTool:Exec
    Crossbow:Exec
    C4:Exec
    RemoteControl:Exec
    DualWield:Exec
    AR_BullPup:Exec
    AR_ForwardGrip:Exec
    MedPackPaddles:Exec
    SMG_P90:Exec
    AR_DrumGun:Exec
    Consumable_Small:Exec
    Consumable_Large:Exec
    Balloon:Exec
    MountedTurret:Exec
    CreativeTool:Exec
    ExplosiveBow:Exec
    AshtonIndigo:Exec
    AshtonChicago:Exec
    MeleeDualWield:Exec
    Unarmed:Exec
   )
   Set Back X Rot Alpha(execute=[@n12.Melee, @n12.MeleeOneHand], Back X Rot Alpha=1)
   n13 = BreakTransform(InTransform=GetDeltaTransformFromRefPose(self=GetOwnerSkeletalMeshComponent(), BoneName=spine_05, BaseName=pelvis))
   MapRangeClamped(Value="Absolute (Float)"(A=BreakRotator(InRot=@n13.Rotation)), InRangeA=5, InRangeB=35)
   TryGetPawnOwner()
  ]}
]
variables [
 arm_up_l {type=number}
 arm_up_r {type=number}
 Back X Rot Alpha {type=number}
 Crouch Alpha {type=number}
 Crouching Neck Alpha {type=number}
 DBNO Alpha {type=number}
 jumping alpha {type=number}
]
